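# Animation Blueprint for the M_MED_BiomeHose_Husk character body cosmetic, copying its pose from the attached parent mesh.
# A bool blend selects between the plain copied pose and a second copy processed by the body's Control Rig.
# A BindPose flag defaults to true and is cleared when character customization loads, though the graph shows no node reading it.
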AnimBlueprint /BRCosmetics/Characters/Player/Male/Medium/Bodies/M_MED_BiomeHose_Husk/Meshes/M_MED_BiomeHose_Husk_AnimBP
target_skeleton=&/M_MED_BiomeHose_Husk_Skeleton
graphs [
 AnimGraph {nodes [
   n1 = Control Rig!(Source=Copy Pose From Mesh() {bUseAttachedParent=true}, Alpha, target:bool=false) -> (Pose) {controlRigAssetReference {blueprintRigClass=&/M_MED_BiomeHose_Husk_CtrlRig.M_MED_BiomeHose_Husk_CtrlRig_C}}
   Output Pose(Result=Blend Poses by bool(BlendPose_0=Copy Pose From Mesh() {bUseAttachedParent=true}, BlendPose_1=@n1)) {layerGroup=None}
  ]}
 EventGraph {nodes [
   Event OnCharacterCustomizationLoaded()
   > Set BindPose(BindPose:bool=false) -> (Output_Get:bool)
   Event BlueprintUpdateAnimation()
  ]}
]
variables [BindPose {type=boolean, default=true}]
