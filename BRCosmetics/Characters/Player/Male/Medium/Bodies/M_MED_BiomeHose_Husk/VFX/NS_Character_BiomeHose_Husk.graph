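# Niagara system for the BiomeHose Husk character cosmetic, building glowing flame hair with sparks.
# Burst-spawned hair, sideburn and scalp meshes follow the skeletal mesh, with time-smoothed positions fed to materials, alongside a glow sprite and particle light.
# One sprite emitter uses three spawn groups for flames and sparks, varying with owner speed and blending between world and local space.
# Colors adapt to time of day, some also to downward camera angle.
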NiagaraSystem /BRCosmetics/Characters/Player/Male/Medium/Bodies/M_MED_BiomeHose_Husk/VFX/NS_Character_BiomeHose_Husk
dependencies {
 usedDataInterfaces [
  &NiagaraDataInterfaceCamera
  &NiagaraDataInterfaceSpriteRendererInfo
  &NiagaraDataInterfaceCurve
  &NiagaraDataInterfaceVector2DCurve
  &NiagaraDataInterfaceColorCurve
  &NiagaraDataInterfaceSkeletalMesh
  &NiagaraDataInterfaceArrayFloat3
 ]
 usedDynamicInputs [&Multiply_Vector, &NormalizeVector]
 usedModules [
  &self:SystemSpawnScript.SystemScriptSource.SystemScriptGraph.NiagaraNodeAssignment_0.SetVariables_#1
  &SystemState
  &BlendSpaceHelper
  &self:SystemSpawnScript.SystemScriptSource.SystemScriptGraph.NiagaraNodeAssignment_2.SetVariables_#2
  &FN_VariantManager
  &self:SystemSpawnScript.SystemScriptSource.SystemScriptGraph.NiagaraNodeAssignment_1.SetVariables_#3
 ]
}
emitters {
 Flame_Spark_Sprites {
  emitterUpdateScript [
   &EmitterState(Life Cycle Mode=::NewEnumerator1, Loop Duration=2, UseLoopDelay=false)
   /SetVariables(Emitter.float:NiagaraFloat=&Waveform("Waveform [1]"=::NewEnumerator0, "[1] Frequency"=1, "[1] Amplitude Min/Max"=<0.1 0.5>)) {isSetParameters=true}
   &SpawnPerUnit(
    Max Movement Threshold=1000
    Module.Use Max Movement Threshold=true
    Module.Use Movement Tolerance=false
    Spawn Spacing=100
    Use Spawn Probability=false
   )
   &SpawnRate(
    SpawnRate=&MakeCustomFloatFromBool(
     Bool=&SetBoolByFloatComparison(A=&VectorLength(Vector=$Engine.Owner.Velocity), B=10)
     True Float=0
     False Float=&MakeCustomFloatFromBool(
      Bool=&SetBoolByFloatComparison(A=&VectorLength(Vector=$Engine.Owner.Velocity), B=800)
      True Float=0
      False Float=2
     )
    )
    Use Spawn Probability=false
   )
   &SpawnRate(
    SpawnGroup=1
    SpawnRate=&MakeCustomFloatFromBool(
     Bool=&SetBoolByFloatComparison(A=&VectorLength(Vector=$Engine.Owner.Velocity), B=10)
     True Float=11
     False Float=&MakeCustomFloatFromBool(
      Bool=&SetBoolByFloatComparison(A=&VectorLength(Vector=$Engine.Owner.Velocity), B=800)
      True Float=1.5
      False Float=1.8
     )
    )
    Use Spawn Probability=false
   ) {module=SpawnRate001}
   &SpawnRate(
    SpawnGroup=2
    SpawnRate=&MakeCustomFloatFromBool(
     Bool=&SetBoolByFloatComparison(A=&VectorLength(Vector=$Engine.Owner.Velocity), B=10)
     True Float=13
     False Float=&MakeCustomFloatFromBool(
      Bool=&SetBoolByFloatComparison(A=&VectorLength(Vector=$Engine.Owner.Velocity), B=800)
      True Float=1.5
      False Float=1.5
     )
    )
    Use Spawn Probability=false
   ) {module=SpawnRate002}
  ]
  particleSpawnScript [
   /SetVariables(Particles.VisibilityTag=&SpawnGroup()) {isSetParameters=true}
   &InitializeParticle(
    Color Mode=::NewEnumerator0
    Lifetime=&MergeHelper_FloatSwitch(
     0=&MakeCustomFloatFromBool(
      Bool=&SetBoolByFloatComparison(A=&VectorLength(Vector=$Engine.Owner.Velocity), B=10)
      True Float=&RandomRangeFloat(Minimum=0.45, Maximum=0.55)
      False Float=&MakeCustomFloatFromBool(
       Bool=&SetBoolByFloatComparison(A=&VectorLength(Vector=$Engine.Owner.Velocity), B=800)
       True Float=0.25
       False Float=&RandomRangeFloat(Minimum=0.65, Maximum=0.75)
      )
     )
     1=&RandomRangeFloat(Minimum=0.2, Maximum=0.5)
     2=&RandomRangeFloat(Minimum=0.2, Maximum=0.6)
     3=?
     Index=&SpawnGroup()
     Num=::NewEnumerator1
    )
    Mass Mode=::NewEnumerator3
    Sprite Rotation Angle=&MergeHelper_FloatSwitch(
     0=&UniformAOrBFloat(
      A=&RandomRangeFloat(Minimum=-70, Maximum=70)
      B=&RandomRangeFloat(Minimum=360, Maximum=-360)
      Evaluation Type=::NewEnumerator0
      Recalculate Random Each Loop=true
      Override Seed=false
     )
     1=&RandomRangeFloat(Minimum=360, Maximum=-360)
     2=&RandomRangeFloat(Minimum=360, Maximum=-360)
     3=?
     Index=&SpawnGroup()
     Num=::NewEnumerator1
    )
    Sprite Rotation Mode=::NewEnumerator0
    Sprite Size Mode=::NewEnumerator3
    Uniform Sprite Size=&MergeHelper_FloatSwitch(
     0=&RandomRangeFloat(Minimum=15, Maximum=20)
     1=&UniformAOrBFloat(
      A=&RandomRangeFloat(Minimum=0.5, Maximum=1)
      B=1.5
      Distribution Weight=0.3
      Evaluation Type=::NewEnumerator0
      Recalculate Random Each Loop=true
      Override Seed=false
     )
     2=&UniformAOrBFloat(
      A=&RandomRangeFloat(Minimum=0.5, Maximum=1)
      B=1.5
      Distribution Weight=0.3
      Evaluation Type=::NewEnumerator0
      Recalculate Random Each Loop=true
      Override Seed=false
     )
     3=?
     Index=&SpawnGroup()
     Num=::NewEnumerator1
    )
    UsePositionOffset=false
    Write Material Random=true
   )
   &ShapeLocation(
    MaskSpawnGroup=false
    Offset=&MergeHelper_VectorSwitch(
     0=<11 0 0>
     1=&RandomRangeVector(Minimum=<12 0 0>, Maximum=<15 0 0>)
     2=&RandomRangeVector(Minimum=<12 0 0>, Maximum=<15 0 0>)
     3=?
     Index=&SpawnGroup()
     Num=::NewEnumerator1
    )
    Offset Mode=::NewEnumerator0
    Override Seed=false
    Sphere Radius=&MergeHelper_FloatSwitch(0=8, 1=8, 2=15, 3=?, Index=&SpawnGroup(), Num=::NewEnumerator1)
    UseOwnerScale=true
   )
   &AddVelocity(Velocity=&MaskVector(
     Mask Source=0
     Mask Parameter=?
     Mask=0
     Masked Vector=&RandomRangeVector(Minimum=<42 -7 -2>, Maximum=<47 -17 2>)
     Passthrough Vector=<0 0 0>
    ))
   &AddVelocity(
    Cone Angle=32
    Distribution Along Cone Axis=0.25
    Module.UseConeFalloff=true
    Module.UseVelDistribution=true
    Override Seed=false
    Speed Falloff From Cone Axis=0.5
    Velocity Mode=::NewEnumerator2
    Velocity Speed=&MergeHelper_FloatSwitch(
     0=0
     1=&RandomRangeFloat(Minimum=25, Maximum=100)
     2=&RandomRangeFloat(Minimum=25, Maximum=75)
     3=?
     Index=&SpawnGroup()
     Num=::NewEnumerator1
    )
   ) {module=AddVelocity001}
   &SubUVAnimation(
    Output Alternate Renderer Binding=true
    "Sprite Renderer "=DataInterface() {propertyValues {SpriteRenderer=&self:Flame_Spark_Sprites_0.NiagaraSpriteRendererProperties_0}}
    SubUV Animation Mode:ENiagara_SubUVLookupModeV2=::NewEnumerator2
    UseEndFrame=false
    UseStartFrame=false
   )
   &CameraOffset(Camera Offset Amount=&MaskFloat(Mask Source=0, Mask Parameter=?, Mask=0, Masked Float=-5, Passthrough Float=4))
  ]
  particleUpdateScript [
   &ParticleState()
   &ScaleSpriteSize(
    Scale Factor=&MergeHelper_Vector2Switch(
     0=&Vector2DFromFloat(Value=&FloatFromCurve(
       FloatCurve=DataInterface() {propertyValues {Curve {
          defaultValue=0.1334
          keys [{interpMode=RCIM_Cubic, tangentMode=RCTM_User, time=0, value=1}, {interpMode=RCIM_Cubic, tangentMode=RCTM_User, time=1, value=0}]
         }}}
       Scale Curve=0.9
      ))
     1=&Vector2DFromCurve(
      Vector2Curve=DataInterface() {propertyValues {
        XCurve {keys [
          <at 0 0 RCIM_Cubic>
          {arriveTangent=0.4755, interpMode=RCIM_Cubic, leaveTangent=0.4755, time=0.1212, value=1}
          {arriveTangent=-1.1338, interpMode=RCIM_Cubic, leaveTangent=-1.1338, time=0.5976, value=0.2841}
          <at 1.0032 0 RCIM_Cubic>
         ]}
        YCurve {keys [<at 0 1 RCIM_Cubic>, {arriveTangent=-1, interpMode=RCIM_Cubic, leaveTangent=-1, time=0.2665, value=1.1362}, <at 1 0 RCIM_Cubic>]}
       }}
      ScaleCurve=<1 1>
     )
     2=&Vector2DFromFloat(Value=&FloatFromCurve(FloatCurve=DataInterface() {propertyValues {Curve {
          defaultValue=0
          keys [
           {interpMode=RCIM_Cubic, tangentMode=RCTM_User, time=0, value=0}
           {interpMode=RCIM_Cubic, tangentMode=RCTM_User, time=0.2635, value=1}
           {arriveTangent=-1.3578, interpMode=RCIM_Cubic, leaveTangent=-1.3578, time=0.6776, value=0.8618}
           {tangentMode=RCTM_User, time=1, value=0}
          ]
         }}}))
     3=?
     4=?
     Index=&SpawnGroup()
     Num=::NewEnumerator1
    )
    Scale Sprite Size Mode=::NewEnumerator2
   )
   &InheritVelocity(
    Inherited Velocity Amount Scale=&MergeHelper_VectorSwitch(
     0=<0 0 0>
     1=&MakeCustomVectorFromBool(
      Bool=&SetBoolByFloatComparison(A=&VectorLength(Vector=$Engine.Owner.Velocity), B=600)
      True Vector=<0 0 0>
      False Vector=<0 0.01 0>
     )
     2=&MakeCustomVectorFromBool(
      Bool=&SetBoolByFloatComparison(A=&VectorLength(Vector=$Engine.Owner.Velocity), B=600)
      True Vector=<0 0 0>
      False Vector=<0 0.01 0>
     )
     3=?
     Index=&SpawnGroup()
     Num=::NewEnumerator1
    )
    Speed Limit=true
    UseSpeedThreshold=false
   )
   &SpriteRotationRate(Rotation Rate=&MergeHelper_FloatSwitch(
     0=&RandomRangeFloat(Minimum=-20, Maximum=20)
     1=0
     2=0
     3=?
     Index=&SpawnGroup()
     Num=::NewEnumerator1
    ))
   &ScaleColor(
    Scale Alpha=&MergeHelper_FloatSwitch(
     0=&FloatFromCurve(
      FloatCurve=DataInterface() {propertyValues {Curve {keys [
          <at -0.0046 1.0113 RCIM_Cubic>
          {arriveTangent=-0.9785, interpMode=RCIM_Cubic, leaveTangent=-0.9785, time=0.5363, value=0.7606}
          {arriveTangent=-1.6322, interpMode=RCIM_Cubic, leaveTangent=-1.6322, time=0.8542, value=0.171}
          <at 1.0011 0.002 RCIM_Cubic>
         ]}}}
      Scale Curve=0.5
     )
     1=&FloatFromCurve(
      FloatCurve=DataInterface() {propertyValues {Curve {
         defaultValue=1
         keys [
          {tangentMode=RCTM_User, time=0, value=1}
          {arriveTangent=-0.0405, interpMode=RCIM_Cubic, leaveTangent=-0.0405, tangentMode=RCTM_User, time=0.7009, value=1}
          {arriveTangent=-8.0089, interpMode=RCIM_Cubic, leaveTangent=-8.0089, tangentMode=RCTM_User, time=1, value=0}
         ]
        }}}
      Scale Curve=2
     )
     2=&FloatFromCurve(
      FloatCurve=DataInterface() {propertyValues {Curve {
         defaultValue=1
         keys [
          {tangentMode=RCTM_User, time=0, value=1}
          {arriveTangent=-0.0405, interpMode=RCIM_Cubic, leaveTangent=-0.0405, tangentMode=RCTM_User, time=0.7009, value=1}
          {arriveTangent=-8.0089, interpMode=RCIM_Cubic, leaveTangent=-8.0089, tangentMode=RCTM_User, time=1, value=0}
         ]
        }}}
      Scale Curve=2
     )
     3=?
     Index=&SpawnGroup()
     Num=::NewEnumerator1
    )
    Scale RGB=&MergeHelper_VectorSwitch(
     0=&MakeVectorFromLinearColorRGB(Color=&FNTimeOfDayColor(
       Night Color=&ColorFromCurve(ColorCurve=DataInterface() {propertyValues {
          AlphaCurve {keys []}
          BlueCurve {keys [<at 0.0327 0>,<at 0.1358 0.2844>,<at 0.2482 0>,<at 0.4658 0>,<at 0.6743 0>]}
          GreenCurve {keys [<at 0.0327 1>,<at 0.1358 1>,<at 0.2482 1>,<at 0.4658 0.4621>,<at 0.6743 0.2346>]}
          RedCurve {keys [<at 0.0327 0.751>,<at 0.1358 0.9411>,<at 0.2482 0.9702>,<at 0.4658 1>,<at 0.6743 1>]}
         }})
       Morning Color=&ColorFromCurve(ColorCurve=DataInterface() {propertyValues {
          AlphaCurve {keys []}
          BlueCurve {keys [<at 0.0139 0.3473>,<at 0.3013 0.0219>,<at 0.4965 0>,<at 0.6354 0>]}
          GreenCurve {keys [<at 0.0139 1>,<at 0.3013 0.8728>,<at 0.4965 0.4612>,<at 0.6354 0.2359>]}
          RedCurve {keys [<at 0.0139 0.9908>,<at 0.3013 1>,<at 0.4965 1>,<at 0.6354 1>]}
         }})
       Day Color=&ColorFromCurve(ColorCurve=DataInterface() {propertyValues {
          AlphaCurve {keys []}
          BlueCurve {keys [<at 0.0279 3.4323>,<at 0.1184 0.2451>,<at 0.2322 0.1844>,<at 0.4278 0>,<at 0.6321 0>,<at 0.8102 0>]}
          GreenCurve {keys [<at 0.0279 4.9486>,<at 0.1184 0.9054>,<at 0.2322 0.9302>,<at 0.4278 0.5458>,<at 0.6321 0.4785>,<at 0.8102 0.2359>]}
          RedCurve {keys [<at 0.0279 5>,<at 0.1184 1>,<at 0.2322 1>,<at 0.4278 0.891>,<at 0.6321 1>,<at 0.8102 1>]}
         }})
       Evening Color=&ColorFromCurve(ColorCurve=DataInterface() {propertyValues {
          AlphaCurve {keys []}
          BlueCurve {keys [<at 0.0076 0.0622>,<at 0.108 0.2885>,<at 0.288 0>,<at 0.4106 0>,<at 0.6283 0>]}
          GreenCurve {keys [<at 0.0076 0.981>,<at 0.108 0.9075>,<at 0.288 0.8328>,<at 0.4106 0.4612>,<at 0.6283 0.2359>]}
          RedCurve {keys [<at 0.0076 1>,<at 0.108 1>,<at 0.288 1>,<at 0.4106 1>,<at 0.6283 1>]}
         }})
      ))
     1=&MakeVectorFromLinearColorRGB(Color=&RandomRangeLinearColor(
       Minimum=&ColorFromCurve(ColorCurve=DataInterface() {propertyValues {
          AlphaCurve {keys []}
          BlueCurve {keys [<at 0 0.8263>,<at 0.8472 0>,<at 1 0.0007>]}
          GreenCurve {keys [<at 0 1.8801>,<at 0.8472 0.8791>,<at 1 0.1885>]}
          RedCurve {keys [<at 0 2>,<at 0.8472 1>,<at 1 1>]}
         }})
       Maximum=&ColorFromCurve(ColorCurve=DataInterface() {propertyValues {
          AlphaCurve {keys []}
          BlueCurve {keys [<at 0.1023 0.8263>,<at 0.5877 0>,<at 1 0.0329>]}
          GreenCurve {keys [<at 0.1023 1.8801>,<at 0.5877 0.8791>,<at 1 0.1729>]}
          RedCurve {keys [<at 0.1023 2>,<at 0.5877 1>,<at 1 1>]}
         }})
      ))
     2=&MakeVectorFromLinearColorRGB(Color=&RandomRangeLinearColor(
       Minimum=&ColorFromCurve(ColorCurve=DataInterface() {propertyValues {
          AlphaCurve {keys []}
          BlueCurve {keys [<at 0 0.8263>,<at 0.8472 0>,<at 1 0.0007>]}
          GreenCurve {keys [<at 0 1.8801>,<at 0.8472 0.8791>,<at 1 0.1885>]}
          RedCurve {keys [<at 0 2>,<at 0.8472 1>,<at 1 1>]}
         }})
       Maximum=&ColorFromCurve(ColorCurve=DataInterface() {propertyValues {
          AlphaCurve {keys []}
          BlueCurve {keys [<at 0.1023 0.8263>,<at 0.5877 0>,<at 1 0.0329>]}
          GreenCurve {keys [<at 0.1023 1.8801>,<at 0.5877 0.8791>,<at 1 0.1729>]}
          RedCurve {keys [<at 0.1023 2>,<at 0.5877 1>,<at 1 1>]}
         }})
      ))
     3=?
     Index=&SpawnGroup()
     Num=::NewEnumerator1
    )
    ScaleA=true
    ScaleRGB=true
   )
   &GravityForce(Gravity=&MaskVector(
     Mask Source=0
     Mask Parameter=?
     Mask=2
     Masked Vector=&MakeCustomVectorFromBool(
      Bool=&SetBoolByFloatComparison(A=&VectorLength(Vector=$Engine.Owner.Velocity), B=20)
      True Vector=<0 0 20>
      False Vector=&MakeCustomVectorFromBool(
       Bool=&SetBoolByFloatComparison(A=&VectorLength(Vector=$Engine.Owner.Velocity), B=800)
       True Vector=<0 0 0>
       False Vector=<0 0 0>
      )
     )
     Passthrough Vector=<0 0 0>
    ))
   &CurlNoiseForce(
    Bias Noise=false
    Noise Frequency=&MergeHelper_FloatSwitch(
     0=25
     1=&RandomRangeFloat(Minimum=30, Maximum=75)
     2=80
     3=?
     Index=&SpawnGroup()
     Num=::NewEnumerator1
    )
    Noise Strength=&MergeHelper_FloatSwitch(
     0=&UniformAOrBFloat(
      A=30
      B=&RandomRangeFloat(Minimum=50, Maximum=70)
      Evaluation Type=::NewEnumerator0
      Recalculate Random Each Loop=true
      Override Seed=false
     )
     1=&RandomRangeFloat(Minimum=35, Maximum=50)
     2=&RandomRangeFloat(Minimum=100, Maximum=200)
     3=?
     Index=&SpawnGroup()
     Num=::NewEnumerator1
    )
    Pan Noise=false
   )
   &Drag(Use Linear Drag=true, Use Rotational Drag=true)
   &SolveForcesAndVelocity(Clamp Velocity=false, Limit Acceleration=false)
   /SetVariables(
    Particles.Flames Outline Color:LinearColor=&MakeLinearColorFromVectorAndFloat(
     "Vector (RGB)"=&MakeVectorFromLinearColorRGB(Color=&FNTimeOfDayColor(
       Night Color=&ColorFromCurve(ColorCurve=DataInterface() {propertyValues {
          AlphaCurve {keys []}
          BlueCurve {keys [<at 0.0597 0>,<at 0.4776 0>,<at 0.8371 0.003>]}
          GreenCurve {keys [<at 0.0597 0.6852>,<at 0.4776 0.1985>,<at 0.8371 0.1019>]}
          RedCurve {keys [<at 0.0597 1>,<at 0.4776 0.8229>,<at 0.8371 0.5104>]}
         }})
       Morning Color=&ColorFromCurve(ColorCurve=DataInterface() {propertyValues {
          AlphaCurve {keys []}
          BlueCurve {keys [<at 0.0597 0>,<at 0.5891 0>,<at 0.8371 0.003>]}
          GreenCurve {keys [<at 0.0597 0.2841>,<at 0.5891 0.1985>,<at 0.8371 0.1019>]}
          RedCurve {keys [<at 0.0597 1>,<at 0.5891 0.8229>,<at 0.8371 0.5104>]}
         }})
       Day Color=&ColorFromCurve(ColorCurve=DataInterface() {propertyValues {
          AlphaCurve {keys []}
          BlueCurve {keys [<at 0.0597 0>,<at 0.5891 0>,<at 0.8371 0.003>]}
          GreenCurve {keys [<at 0.0597 0.2841>,<at 0.5891 0.1985>,<at 0.8371 0.1019>]}
          RedCurve {keys [<at 0.0597 1>,<at 0.5891 0.8229>,<at 0.8371 0.5104>]}
         }})
       Evening Color=&ColorFromCurve(ColorCurve=DataInterface() {propertyValues {
          AlphaCurve {keys []}
          BlueCurve {keys [<at 0.0597 0>,<at 0.5891 0>,<at 0.8371 0.003>]}
          GreenCurve {keys [<at 0.0597 0.2841>,<at 0.5891 0.1985>,<at 0.8371 0.1019>]}
          RedCurve {keys [<at 0.0597 1>,<at 0.5891 0.8229>,<at 0.8371 0.5104>]}
         }})
      ))
     "Float (Alpha)"=&FloatFromCurve(
      FloatCurve=DataInterface() {propertyValues {Curve {keys [
          <at -0.0046 1.0113 RCIM_Cubic>
          {arriveTangent=-0.6404, interpMode=RCIM_Cubic, leaveTangent=-0.6404, time=0.3557, value=0.8556}
          {arriveTangent=-1.3696, interpMode=RCIM_Cubic, leaveTangent=-1.3696, time=0.6846, value=0.57}
          {arriveTangent=-1.7943, interpMode=RCIM_Cubic, leaveTangent=-1.7943, time=0.8624, value=0.1617}
          <at 1.0011 0.002 RCIM_Cubic>
         ]}}}
      Scale Curve=0.6
     )
    )
    Particles.Flames Outline Size:Vector2f=&Multiply_Vector2D_ByFloat(Float=1.3, Vector2D=$Particles.SpriteSize)
   ) {isSetParameters=true}
   /SetVariables(
    Particles.Sparks Outline COlor:LinearColor=&MakeLinearColorFromVectorAndFloat(
     "Vector (RGB)"=&MakeVectorFromLinearColorRGB(Color=&ColorFromCurve(ColorCurve=DataInterface() {propertyValues {
         AlphaCurve {keys []}
         BlueCurve {keys [
           <at 0.0291 0.6308 RCIM_Cubic>
           <at 0.2001 0>
           {arriveTangent=-0.8596, interpMode=RCIM_Cubic, leaveTangent=-0.8596, time=0.376, value=0}
           <at 0.7439 0.0163>
          ]}
         GreenCurve {keys [
           <at 0.0291 0.9619 RCIM_Cubic>
           <at 0.2001 0.7244>
           {arriveTangent=-1.2461, interpMode=RCIM_Cubic, leaveTangent=-1.2461, time=0.376, value=0.2953}
           <at 0.7439 0.0712>
          ]}
         RedCurve {keys [
           <at 0.0291 1 RCIM_Cubic>
           <at 0.2001 1>
           {arriveTangent=-0.8175, interpMode=RCIM_Cubic, leaveTangent=-0.8175, time=0.376, value=1}
           <at 0.7439 0.4156>
          ]}
        }}))
     "Float (Alpha)"=&FloatFromCurve(
      FloatCurve=DataInterface() {propertyValues {Curve {
         defaultValue=1
         keys [
          {tangentMode=RCTM_User, time=0, value=1}
          {arriveTangent=-0.0405, interpMode=RCIM_Cubic, leaveTangent=-0.0405, tangentMode=RCTM_User, time=0.7009, value=1}
          {arriveTangent=-8.0089, interpMode=RCIM_Cubic, leaveTangent=-8.0089, tangentMode=RCTM_User, time=1, value=0}
         ]
        }}}
      Scale Curve=1.25
     )
    )
    Particles.Sparks Outline Size:Vector2f=&Multiply_Vector2D_ByFloat(Float=2, Vector2D=$Particles.SpriteSize)
   ) {isSetParameters=true}
   /SetVariables(
    Particles.Linger Sparks Outline Color:LinearColor=&MakeLinearColorFromVectorAndFloat(
     "Vector (RGB)"=&MakeVectorFromLinearColorRGB(Color=&ColorFromCurve(ColorCurve=DataInterface() {propertyValues {
         AlphaCurve {keys []}
         BlueCurve {keys [
           <at 0.0291 0.6308 RCIM_Cubic>
           <at 0.3662 0>
           {arriveTangent=-0.7302, interpMode=RCIM_Cubic, leaveTangent=-0.7302, time=0.6418, value=0}
           <at 0.8706 0.0163>
          ]}
         GreenCurve {keys [
           <at 0.0291 0.9619 RCIM_Cubic>
           <at 0.3662 0.7244>
           {arriveTangent=-1.0584, interpMode=RCIM_Cubic, leaveTangent=-1.0584, time=0.6418, value=0.2953}
           <at 0.8706 0.0712>
          ]}
         RedCurve {keys [
           <at 0.0291 1 RCIM_Cubic>
           <at 0.3662 1>
           {arriveTangent=-0.6944, interpMode=RCIM_Cubic, leaveTangent=-0.6944, time=0.6418, value=1}
           <at 0.8706 0.4156>
          ]}
        }}))
     "Float (Alpha)"=&FloatFromCurve(
      FloatCurve=DataInterface() {propertyValues {Curve {
         defaultValue=1
         keys [
          {tangentMode=RCTM_User, time=0, value=1}
          {arriveTangent=-0.0405, interpMode=RCIM_Cubic, leaveTangent=-0.0405, tangentMode=RCTM_User, time=0.7009, value=1}
          {arriveTangent=-8.0089, interpMode=RCIM_Cubic, leaveTangent=-8.0089, tangentMode=RCTM_User, time=1, value=0}
         ]
        }}}
      Scale Curve=1.25
     )
    )
    Particles.Linger Sparks Outline Size:Vector2f=&Multiply_Vector2D_ByFloat(Float=2, Vector2D=$Particles.SpriteSize)
   ) {isSetParameters=true}
   /SetVariables(
    Particles.Glow Color=&MakeLinearColorFromVectorAndFloat(
     "Vector (RGB)"=&MakeVectorFromLinearColorRGB(Color=<rgba 1 0.3802 0 1>)
     "Float (Alpha)"=&FloatFromCurve(FloatCurve=DataInterface() {propertyValues {Curve {
         defaultValue=1
         keys [{interpMode=RCIM_Cubic, tangentMode=RCTM_User, time=0, value=1}, {interpMode=RCIM_Cubic, tangentMode=RCTM_User, time=1, value=0}]
        }}})
    )
    Particles.Glow Size=&Multiply_Vector2D_ByFloat(Float=3.15, Vector2D=$Particles.SpriteSize)
   ) {isSetParameters=true}
   &BlendSpace("Position Blend (World <-> Local)":NiagaraFloat=&MergeHelper_FloatSwitch(
     0=&FloatFromCurve(FloatCurve=DataInterface() {propertyValues {Curve {keys [
          <at 0.15 1 RCIM_Cubic>
          {arriveTangent=-0.1839, interpMode=RCIM_Cubic, leaveTangent=-0.1839, time=0.5205, value=0.8796}
          <at 0.9525 0.8524 RCIM_Cubic>
         ]}}})
     1=&FloatFromCurve(FloatCurve=DataInterface() {propertyValues {Curve {keys [<at 0.1543 1>,<at 0.5836 0.9605>,<at 1.0012 0.6926>]}}})
     2=&FloatFromCurve(FloatCurve=DataInterface() {propertyValues {Curve {keys [<at 0.3575 0.7993>,<at 0.4617 0.6675>,<at 1.0012 0.5906>]}}})
     3=?
     Index=&SpawnGroup()
     Num=::NewEnumerator1
     Translation Contribution:NiagaraFloat=1
     Rotation Contribution:NiagaraFloat=1
    ))
   /SetVariables(Particles.Flame Licks Dynamic Param:Vector4f=&MakeVector4(
     X=&FloatFromCurve(
      FloatCurve=DataInterface() {propertyValues {Curve {
         defaultValue=0
         keys [
          {interpMode=RCIM_Cubic, tangentMode=RCTM_User, time=0, value=1}
          {arriveTangent=-1, interpMode=RCIM_Cubic, leaveTangent=-1, time=0.1736, value=0.4609}
          {interpMode=RCIM_Cubic, tangentMode=RCTM_User, time=1, value=0}
         ]
        }}}
      Scale Curve=0.5
     )
     Y=0
     Z=0
     W=0
    )) {isSetParameters=true}
  ]
  rendererClasses [&NiagaraSpriteRendererProperties]
  renderers [
   {rendererClass=&NiagaraSpriteRendererProperties}
   {rendererClass=&NiagaraSpriteRendererProperties, rendererIndex=1}
   {rendererClass=&NiagaraSpriteRendererProperties, rendererIndex=2}
   {rendererClass=&NiagaraSpriteRendererProperties, rendererIndex=3}
   {rendererClass=&NiagaraSpriteRendererProperties, rendererIndex=4}
   {rendererClass=&NiagaraSpriteRendererProperties, rendererIndex=5}
   {rendererClass=&NiagaraSpriteRendererProperties, rendererIndex=6}
   {rendererClass=&NiagaraSpriteRendererProperties, rendererIndex=7}
  ]
 }
 FlowingHair {
  emitterUpdateScript [
   &EmitterState(
    Inactive Response=::NewEnumerator1
    Life Cycle Mode=::NewEnumerator1
    Loop Behavior=::NewEnumerator1
    Loop Duration Mode=::NewEnumerator1
    UseLoopDelay=false
   )
   &SpawnBurst_Instantaneous(Use Spawn Probability=false, UseLoopCountLimit=false)
  ]
  particleSpawnScript [
   &InitializeParticle(
    Color Mode=::NewEnumerator0
    Lifetime=2
    Mesh Scale=<0.38 0.38 0.19>
    Mesh Scale Mode=::NewEnumerator4
    Sprite Size Mode=::NewEnumerator3
    UsePositionOffset=false
    Write Material Random=true
   )
   &InitialMeshOrientation(Mesh Axis To Orient=<0 0 1>, Mesh Orientation Mode=::NewEnumerator2, Rotation Coordinate Space=::NewEnumerator0, Use Rotation Vector=true)
   &DynamicMaterialParameters(
    Index 0 Param 2=&RandomRangeFloat(Minimum=0, Maximum=1.5)
    Index 0 Param 4=1
    Index 2 Param 2=20
    Index 2 Param 4=1
    Module.Index0Float0Write=true
    Module.Index0Float1Write=true
    Module.Index0Float2Write=true
    Module.Index0Float3Write=true
    Module.Index2Float0Write=true
    Module.Index2Float1Write=true
    Module.Index2Float2Write=true
    Module.Index2Float3Write=true
    Write Parameter Index 2=::NewEnumerator1
   )
   &CameraOffset(Camera Offset Amount=-1)
  ]
  particleUpdateScript [
   &ParticleState(
    Kill Particles When Lifetime Has Elapsed=false
    Let Infinitely Lived Particles Die When Emitter Deactivates=true
    Loop Particles Lifetime=true
   )
   &SkeletalMeshLocation(
    Bone Sampling Mode=::NewEnumerator4
    MaskSpawnGroup=false
    Orientation Sampling=::NewEnumerator1
    Sampled Position Offset=<7 0 0>
    Skeletal Mesh=$System.Skeletal Mesh
   )
   &UpdateMeshOrientation(
    Facing Direction=&Add_Vector(
     A=<0 0 1>
     B=&RotateVectorByQuaternion(Vector To Rotate=<0 -0.15 0>, Quaternion=$Particles.SkeletalMeshLocation.SampledRotation)
    )
    Facing Interpolation=false
    Orientation Method=::NewEnumerator1
    Override Quat=false
    TransformFacing=false
    TransformUp=false
   ) {module=UpdateMeshOrientation002}
   &UpdateMeshOrientation(Rotation Rate=-0.2, Rotation Vector=<0 0 1>)
   &InterpolateOverTime(
    Rate Of Change=5
    Target Vector Value:Vector3f=&ConvertPositionToVector(
     Input Position=&ConvertVectorToPosition(
      Input Position=Linked() {linkedVariable {name=Particles.SkeletalMeshLocation.SampledPosition, type=NiagaraPosition}}
      Vector Is In Absolute Large World Space:NiagaraBool=true
     )
     Convert To Absolute Large World Vector=false
    )
    Variable Type=::NewEnumerator2
   )
   /SetVariables(Particles.InterpolateDynamic:Vector4f=&MakeFloat4FromVectorAndFloat("Vector (RGB)"=$Particles.InterpolateOverTime.InterpolateOverTime_V3_.Moving Average, "Float (Alpha)"=1.5)) {isSetParameters=true}
   &SolveForcesAndVelocity(Clamp Velocity=false, Limit Acceleration=false)
   /SetVariables(
    Particles.Linear Color=&MakeLinearColorFromVectorAndFloat(
     "Vector (RGB)"=&MakeVectorFromLinearColorRGB(Color=&FNTimeOfDayColor(
       Night Color=<rgba 3 2.7214 0.4217 2>
       Morning Color=<rgba 3 2.7214 0.4217 2>
       Day Color=<rgba 3 2.7214 0.4217 2>
       Evening Color=<rgba 3 2.7214 0.4217 2>
      ))
     "Float (Alpha)"=&Lerp_Float(
      A=&FNTimeOfDayScalar(Night Scale=1, Morning Scale=2, Day Scale=2, Evening Scale=2)
      B=0.25
      Alpha=&ClampFloat(Float=&RemapRange(
        Input Value To Remap=&DotProduct(Vector A=&ReturnCameraVector(Camera Vectors=::NewEnumerator1, Coordinate Space=::World), Vector B=<0 0 -1>)
        Input Min=0.7
        Input Max=0.9
        Output Min=0
        Output Max=1
       ))
     )
    )
    Particles.Vector:Vector3f=&Multiply_VectorByFloat(Vector=$Particles.Scale, Float=1.1)
   ) {isSetParameters=true}
   &ScaleMeshSize(Scale Factor=&VectorFromFloat(Value=&Lerp_Float(A=1, B=1.8, Alpha=$NPC.FortniteNPC.BigHeadMode)))
   &FN_Light_Attributes(
    Color=<rgba 20 7.9365 0 1>
    Create Color Attribute=true
    Create Exponent Attribute=true
    Create Position Attribute=true
    Create Radius Attribute=true
    Enable Particle Light Enable=true
    Exponent=200
    Light Enabled=true
    Position:NiagaraPosition=$Output.SkeletalMeshLocation.SampledPositionWithoutOffset
    Radius=90
    Scale Color=1.5
   )
  ]
  rendererClasses [&NiagaraMeshRendererProperties, &NiagaraLightRendererProperties]
  renderers [
   {rendererClass=&NiagaraMeshRendererProperties, rendererIndex=0}
   {rendererClass=&NiagaraLightRendererProperties, rendererIndex=1}
  ]
 }
 Forehead_Scalp {
  emitterUpdateScript [
   &EmitterState(
    Inactive Response=::NewEnumerator1
    Life Cycle Mode=::NewEnumerator1
    Loop Behavior=::NewEnumerator1
    Loop Duration Mode=::NewEnumerator1
    UseLoopDelay=false
   )
   &SpawnBurst_Instantaneous(Use Spawn Probability=false, UseLoopCountLimit=false)
   &SpawnBurst_Instantaneous(Spawn Group=1, Use Spawn Probability=false, UseLoopCountLimit=false) {module=SpawnBurst_Instantaneous001}
   /SetVariables(
    Emitter.LeadColor:LinearColor=&FNTimeOfDayColor(
     Night Color=<rgba 1 0.989 0.0978 1>
     Morning Color=<rgba 1 0.9274 0.156 1>
     Day Color=<rgba 1 0.7969 0 1>
     Evening Color=<rgba 1 0.9274 0.156 1>
    )
    Emitter.MainColor:LinearColor=&FNTimeOfDayColor(
     Night Color=<rgba 0.9168 1.2229 0.3376 1>
     Morning Color=<rgba 1.1 0.9066 0.3266 1>
     Day Color=<rgba 3 2.9413 0.5781 1>
     Evening Color=<rgba 1.1 1.0616 0.5039 1>
    )
    Emitter.TexAColor1:LinearColor=&LerpLinearColors(
     StartColor=&FNTimeOfDayColor(
      Night Color=<rgba 1 0.8302 0.4613 1>
      Morning Color=<rgba 1.75 1.404 0.0071 1>
      Day Color=<rgba 1.75 0.3958 0 1>
      Evening Color=<rgba 1.75 1.6681 0.223 1>
     )
     EndColor=&FNTimeOfDayColor(
      Night Color=<rgba 3 2.9207 0.7543 1>
      Morning Color=<rgba 2 1.4849 0.5351 1>
      Day Color=<rgba 2 1.4849 0.5351 1>
      Evening Color=<rgba 1.8893 2 0.3953 1>
     )
     LerpFactor=&ClampFloat(Float=&RemapRange(
       Input Value To Remap=&DotProduct(Vector A=&ReturnCameraVector(Camera Vectors=::NewEnumerator1, Coordinate Space=::World), Vector B=<0 0 -1>)
       Input Min=0.7
       Input Max=0.9
       Output Min=0
       Output Max=1
      ))
    )
    Emitter.TexAColor2:LinearColor=&FNTimeOfDayColor(
     Night Color=<rgba 1.5886 2 0.5687 1>
     Morning Color=<rgba 2 1.7686 0.7067 1>
     Day Color=<rgba 2 1.7686 0.7067 1>
     Evening Color=<rgba 2 1.7686 0.7067 1>
    )
   ) {isSetParameters=true}
  ]
  particleSpawnScript [
   /SetVariables(Particles.VisibilityTag=&SpawnGroup()) {isSetParameters=true}
   &InitializeParticle(
    Color=<rgba 0.6365 0.6365 0.6365 2>
    Mesh Scale=&MaskVector(Mask Source=0, Mask Parameter=?, Mask=0, Masked Vector=<1 1 1>, Passthrough Vector=<2 0.5 2>)
    Mesh Scale Mode=::NewEnumerator4
    Position Offset=&MergeHelper_VectorSwitch(0=<-2 0 0>, 1=<-5 5 0>, 2=?, 3=?, Index=&SpawnGroup(), Num=::NewEnumerator0)
    UsePositionOffset=false
    Write Material Random=true
   )
   /SetVariables(Particles.Secondary Size:Vector3f=<0.9 0.9 0.9>) {isSetParameters=true}
   &CameraOffset(Camera Offset Amount=&MaskFloat(Mask Source=0, Mask Parameter=?, Mask=0, Masked Float=0, Passthrough Float=-1))
  ]
  particleUpdateScript [
   &ParticleState(Kill Particles When Lifetime Has Elapsed=false)
   &ScaleMeshSize(Scale Factor=&VectorFromFloat(Value=&Lerp_Float(A=1, B=1, Alpha=$NPC.FortniteNPC.BigHeadMode)))
   &SkeletalMeshLocation(
    Bone Sampling Mode=::NewEnumerator4
    MaskSpawnGroup=false
    Orientation Sampling=::NewEnumerator1
    Position Sampling=::NewEnumerator2
    Sampled Position Offset Coordinate Space:ENiagaraSkelMeshTransforms=::NewEnumerator2
    Skeletal Mesh=$System.Skeletal Mesh
   )
   &DynamicMaterialParameters(
    Index 0 Param 1=0.6
    Index 0 Param 2=1.5
    Module.Index0Float0Write=true
    Module.Index0Float1Write=true
    Module.Index0Float2Write=true
    Module.Index0Float3Write=true
   ) {module=DynamicMaterialParameters001}
  ]
  rendererClasses [&NiagaraMeshRendererProperties]
  renderers [
   {rendererClass=&NiagaraMeshRendererProperties, rendererIndex=0}
   {rendererClass=&NiagaraMeshRendererProperties, rendererIndex=1}
  ]
 }
 Glow {
  emitterUpdateScript [
   &EmitterState(
    Inactive Response=::NewEnumerator1
    Life Cycle Mode=::NewEnumerator1
    Loop Behavior=::NewEnumerator1
    Loop Duration Mode=::NewEnumerator1
    UseLoopDelay=false
   )
   &SpawnBurst_Instantaneous(Use Spawn Probability=false, UseLoopCountLimit=false)
  ]
  particleSpawnScript [
   &InitializeParticle(
    Color=&FNTimeOfDayColor(
     Night Color=<rgba 1.7714 2 0 10>
     Morning Color=<rgba 2 1.7342 0.3524 10>
     Day Color=<rgba 2 1.7342 0.3524 10>
     Evening Color=<rgba 2 1.7342 0.3524 10>
    )
    Lifetime=$Emitter.CurrentLoopDuration
    Position Offset=<4 1 0>
    Sprite Size=<22 26>
    Sprite Size Mode=::NewEnumerator4
    UsePositionOffset=true
    Write Material Random=true
   )
   &CameraOffset(Camera Offset Amount=-2)
  ]
  particleUpdateScript [
   &ParticleState(
    Kill Particles When Lifetime Has Elapsed=false
    Let Infinitely Lived Particles Die When Emitter Deactivates=true
    Loop Particles Lifetime=true
   )
   &SolveForcesAndVelocity(Clamp Velocity=false, Limit Acceleration=false)
   /SetVariables() {isSetParameters=true}
  ]
  rendererClasses [&NiagaraSpriteRendererProperties]
  renderers [{rendererClass=&NiagaraSpriteRendererProperties}]
 }
 Sideburns {
  emitterUpdateScript [
   &EmitterState(
    Inactive Response=::NewEnumerator1
    Life Cycle Mode=::NewEnumerator1
    Loop Behavior=::NewEnumerator1
    Loop Duration Mode=::NewEnumerator1
    UseLoopDelay=false
   )
   &SpawnBurst_Instantaneous(Use Spawn Probability=false, UseLoopCountLimit=false)
   /SetVariables(
    Emitter.LeadColor:LinearColor=&FNTimeOfDayColor(Night Color=<rgba 1 0.7969 0 1>, Morning Color=<rgba 1 0.7969 0 1>, Day Color=<rgba 1 0.7969 0 1>, Evening Color=<rgba 1 0.7969 0 1>)
    Emitter.MainColor:LinearColor=&FNTimeOfDayColor(
     Night Color=<rgba 0.9145 1.2229 0.1898 1>
     Morning Color=<rgba 1.1 0.925 0.4 1>
     Day Color=<rgba 2.5 2.4612 0.8999 1>
     Evening Color=<rgba 1.1 1.0523 0.4672 1>
    )
    Emitter.TexAColor1:LinearColor=&FNTimeOfDayColor(
     Night Color=<rgba 1 0.1387 0.045 1>
     Morning Color=<rgba 1 0.0823 0.0666 1>
     Day Color=<rgba 1 0.0616 0 1>
     Evening Color=<rgba 1 0.0823 0.0666 1>
    )
    Emitter.TexAColor2:LinearColor=&FNTimeOfDayColor(
     Night Color=<rgba 1.5886 2 0.5687 1>
     Morning Color=<rgba 2 1.7686 0.7067 1>
     Day Color=<rgba 2 1.7686 0.7067 1>
     Evening Color=<rgba 2 1.7686 0.7067 1>
    )
   ) {isSetParameters=true}
  ]
  particleSpawnScript [
   &InitializeParticle(
    Color=<rgba 0.6365 0.6365 0.6365 2>
    Mesh Scale=<0.6 1 1.1>
    Mesh Scale Mode=::NewEnumerator4
    UsePositionOffset=false
    Write Material Random=true
   )
   &InitialMeshOrientation(Mesh Orientation Mode=::NewEnumerator2, Rotation=<-0.05 0.25 0.05>, Use Rotation Vector=true)
   &DynamicMaterialParameters(
    Index 0 Param 1=0.6
    Index 0 Param 2=1.5
    Index 0 Param 3=0.5
    Index 0 Param 4=1
    Index 2 Param 2=2
    Module.Index0Float0Write=true
    Module.Index0Float1Write=true
    Module.Index0Float2Write=true
    Module.Index0Float3Write=true
    Module.Index2Float0Write=true
    Module.Index2Float1Write=true
    Module.Index2Float2Write=true
    Module.Index2Float3Write=true
    Write Parameter Index 2=::NewEnumerator1
   )
  ]
  particleUpdateScript [
   &ParticleState(Kill Particles When Lifetime Has Elapsed=false)
   &ScaleMeshSize(Scale Factor=&Lerp_Vector(A=&VectorFromFloat(Value=0.9), B=<2 1 1.5>, Alpha=$NPC.FortniteNPC.BigHeadMode))
   &SkeletalMeshLocation(Bone Sampling Mode=::NewEnumerator4, MaskSpawnGroup=false, Sampled Position Offset=<-1 -1 0>, Skeletal Mesh=$System.Skeletal Mesh)
   &InterpolateOverTime(Rate Of Change:NiagaraFloat=$System.InterpolateRate, Target Vector Value:NiagaraPosition=$Particles.Position, Variable Type=::NewEnumerator2)
   /SetVariables(Particles.InterpolateDy:Vector4f=&MakeFloat4FromVectorAndFloat("Vector (RGB)"=$StackContext.InterpolateOverTime.InterpolateOverTime_V3_.Moving Average, "Float (Alpha)"=$System.BendIntensity)) {isSetParameters=true}
   &ScaleColor(
    Scale RGB=&Multiply_VectorByFloat(Vector=<1 1 1>, Float=&FN_FrontEndFloatModifier(Gameplay=1, Front End=0.9))
    ScaleA=true
    ScaleRGB=true
   )
  ]
  rendererClasses [&NiagaraMeshRendererProperties]
  renderers [{rendererClass=&NiagaraMeshRendererProperties, rendererIndex=0}]
 }
}
summary {
 systemName=NS_Character_BiomeHose_Husk
 emitters [
  {emitterName=FlowingHair, rendererClasses [&NiagaraMeshRendererProperties, &NiagaraLightRendererProperties]}
  {emitterName=Sideburns, rendererClasses [&NiagaraMeshRendererProperties]}
  {emitterName=Forehead_Scalp, rendererClasses [&NiagaraMeshRendererProperties]}
  {emitterName=Glow, rendererClasses [&NiagaraSpriteRendererProperties]}
  {emitterName=Flame_Spark_Sprites, rendererClasses [&NiagaraSpriteRendererProperties]}
 ]
 userVariables [
  User.bIsFrontEnd {type=NiagaraBool}
  User.bIsFrontEndPreview {type=NiagaraBool}
 ]
}
